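# Physics asset for the Dragon pickaxe, a two-handed melee cosmetic weapon mesh.
# It defines a single box body on the Melee_TwoHanded_Handle bone, offset and rotated as a rough, thin slab approximation of the weapon's shape.
# There are no constraints, so the physics representation is one rigid piece with no articulated parts.
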PhysicsAsset /BRCosmetics/Weapons/FORT_Melee/Pickaxe_Dragon/Meshes/Dragon_Pickaxe_PhysicsAsset
bodies {Melee_TwoHanded_Handle [{shapeType=Box, extentX=90.6295, extentY=12.8422, extentZ=142.8082, center=<0.8043 -1.0572 28.4132>, rotation=<rot 0.0804 -90 -0.0121>}]}
constraints []
